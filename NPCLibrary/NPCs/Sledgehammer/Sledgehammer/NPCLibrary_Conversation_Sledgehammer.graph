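# Conversation database for the Sledgehammer NPC, a spider enthusiast who greets the player from the NPC's entry point.
# A single speech node gives a default line, or prioritized contextual lines for time since last match, first meeting, high bar count, or Black Widow outfits.
# Line priorities come from named curve rows.
# The speech then links out to the shared NPC Library quests and basic services conversations.
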ConversationDatabase /NPCLibrary/NPCs/Sledgehammer/Sledgehammer/NPCLibrary_Conversation_Sledgehammer
entry_points [{tag=Conversation.Entry.Sledgehammer, destinations [#1]}]
nodes [
 ConversationEntryPointNode_7 {
  id=#1
  class=ConversationEntryPointNode
  outputs [#2]
  properties {entryTag {tagName=Conversation.Entry.Sledgehammer}}
 }
 Speech_C_4 {
  id=#2
  class=Speech_C
  outputs [#3, #4]
  properties {
   nodeName=""
   generalConfig {
    defaultMessage=It's about winning the war. And spiders. I like spiders.
    contextualMessages [
     {
      message="I've learned so many spider facts since you've been gone. "
      contextRequirements [{requirement {class=FortControllerRequirement_CheckTimeSinceLastPlayedMatch}}]
      priority {curve {rowName=Tandem.Speech.Priority.WhileSincePlayed}}
     }
     {
      message=I'm sorry in advance for all the spider stuff. I just like 'em.
      contextRequirements [{requirement {class=FortControllerRequirement_IsFirstEverConversationWithNPC}}]
      priority {curve {rowName=Tandem.Speech.Priority.FirstEverInteraction}}
     }
     {
      message=Whoa. You could singlehandedly fund my spider conservancy.
      contextRequirements [{requirement {class=FortControllerRequirement_HasBars, properties {thresholdTestConfig {quantity {value=3000}}}}}]
      priority {curve {rowName=Tandem.Speech.Priority.HighTotalBars}}
     }
     {
      message="Hey! A fellow arachnid admirer! "
      contextRequirements [{requirement {
         class=FortControllerRequirement_HasCID
         properties {allowedCIds [
           /BRCosmetics/Athena/Items/Cosmetics/Characters/CID_248_Athena_Commando_M_BlackWidow
           /BRCosmetics/Athena/Items/Cosmetics/Characters/CID_249_Athena_Commando_F_BlackWidow
          ]}
        }}]
      priority {curve {rowName=Tandem.Speech.Priority.OutfitReaction}}
     }
    ]
   }
  }
 }
 ConversationLinkNode_0 {id=#4, class=ConversationLinkNode, properties {remoteEntryTag {tagName=Conversation.Entry.BasicServices}}}
 ConversationLinkNode_1 {id=#3, class=ConversationLinkNode, properties {remoteEntryTag {tagName=Conversation.Entry.NPCLibrary.Quests}}}
]
